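# Sound cue that plays the first note sound for a memory game quest in Figment Season 9.
# A single wave player feeds the Memory_Game_Note_01 wave straight to the output at reduced volume.
# It is routed to the UI sound class but uses Season 7 quest attenuation.
# A concurrency override allows one instance at a time, stopping the oldest when a new one starts.
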SoundCue /Figment_S09/Athena/Sounds/Quests/Memory_Game/Memory_Game_Note_01_Cue
root=WavePlayer_1
nodes [{id=WavePlayer_1, properties {soundWaveAssetPtr=&/Memory_Game_Note_01}}]
properties {
 attenuationSettings=/Game/Sounds/Attenuation/Fort_Quest_Season7
 bOverrideConcurrency=true
 maxDistance=2400
 soundClassObject=&Fort_UI
 volumeMultiplier=0.7
 concurrencyOverrides {maxCount=1, resolutionRule=StopOldest}
}
