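# Sound cue for the memory game quest's reset sound, judging by its name and path.
# A mixer layers a reused dart-trap placement sound at full level with a quieter construction-done UI sound.
# It plays in the UI sound class with Season 7 quest attenuation, at slightly reduced volume.
# A concurrency override allows one instance at a time, stopping the oldest on retrigger.
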SoundCue /Figment_S09/Athena/Sounds/Quests/Memory_Game/Memory_Game_Reset_Cue
root=Mixer_15
nodes [
 {id=Mixer_15, children [WavePlayer_40, WavePlayer_104], properties {inputVolume [1,0.4]}}
 {id=WavePlayer_104, properties {soundWaveAssetPtr=/Game/Sounds/Fort_UI_Sounds/Gameplay/Fort_Contructino_Done_02}}
 {id=WavePlayer_40, properties {soundWaveAssetPtr=/Game/Sounds/Fort_Traps/Dart_Trap/WallTrap_Darts_Placed_001}}
]
properties {
 attenuationSettings=/Game/Sounds/Attenuation/Fort_Quest_Season7
 bOverrideConcurrency=true
 maxDistance=2400
 soundClassObject=&Fort_UI
 volumeMultiplier=0.8
 concurrencyOverrides {maxCount=1, resolutionRule=StopOldest}
}
